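# UI material that draws an accolade splash icon: a colored background shape, a white-by-default icon and optional tier stars.
# Static switches pick a circle or rotated rounded-box background and toggle the stars; smoothsteps cut crisp shapes from gradient and texture values.
# Stars tile from a distance-field texture, masked by Tier and shifted by parity to stay centered.
# Icon and star masks blend the colors and combine with background alpha; Anim_Highlight scales icon and stars.
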Material /Game/Athena/Items/Accolades/Materials/M_UI_AccoladeSplashIcon
graph {
 properties {blendMode=BLEND_Translucent, materialDomain=MD_UI, pixelDepthOffsetMode=PDOM_Legacy, usedShadingModels=Default Lit}
 nodes [
  n1 = VectorParameter() {group=Background, parameterName=BGcolor, defaultValue=<rgba 0.0789 0.4612 0.7969 1>}
  n2 = NamedRerouteUsage()
  LinearInterpolate(A=@n1.RGB, B=NamedRerouteUsage(), Alpha=Max(A=NamedRerouteUsage(), B=@n2))
  n3 = ScalarParameter() {defaultValue=0.1, group=Background, parameterName=BGEdge}
  n4 = Add(A=NamedRerouteUsage(), B=AppendVector(A=Constant(), B=ScalarParameter() {group=Background, parameterName=BackgroundOffset}))
  n5 = Rotator(Coordinate=@n4, Time=ScalarParameter() {defaultValue=0.125, group=Background, parameterName=RectangleBackgroundAngle}) {speed=6.2832}
  n6 = Constant() {r=1}
  n7 = ScalarParameter() {defaultValue=0.7, group=Background, parameterName=RectangleBackgroundScale}
  n8 = Add(A=Multiply(A=Subtract(A=@n5) {constB=0.5}, B=Divide(A=@n6, B=@n7))) {constB=0.5}
  n9 = Constant() {r=1}
  n10 = ScalarParameter() {defaultValue=0.15, group=Background, parameterName=RectangleBackgroundRoundEdge}
  n11 = Constant() {r=0.05}
  n12 = SmoothStep(
   Min=@n3
   Max=Add(A=@n3, B=ScalarParameter() {defaultValue=0.01, group=Background, parameterName=BGEdgeSoftness})
   Value=StaticSwitchParameter(
    True=MaterialFunctionCall(UVs=@n4) {materialFunction=&RadialGradientExponential}.RadialGradientExponential
    False=MaterialFunctionCall(UVs=@n8, Size=@n9, CornerRoundness=@n10, GlowMax=@n11) {materialFunction=&MF_UI_Box}.Fill
   ) {defaultValue=true, group=Background, parameterName=UseCircleBackground}
  )
  Max(A=NamedRerouteUsage(), B=Max(A=@n2, B=Multiply(A=@n1.A, B=@n12)))
  n13 = ScalarParameter() {defaultValue=0.2, group=Icon, parameterName=IconEdge}
  n14 = ScalarParameter() {defaultValue=0.01, group=Icon, parameterName=IconEdgeSoftness}
  n15 = ScalarParameter() {defaultValue=1, group=Icon, parameterName=IconScale}
  n16 = ScalarParameter() {defaultValue=1.15, group=Animation, parameterName=Anim_Highlight_Scale}
  n17 = MaterialFunctionCall(
   UVs=NamedRerouteUsage()
   Texture Scale=LinearInterpolate(A=@n15, B=Multiply(A=@n15, B=@n16), Alpha=ScalarParameter() {group=Animation, parameterName=Anim_Highlight})
  ) {materialFunction=&ScaleUVsByCenter}
  n18 = TextureSampleParameter2D(UVs=@n17.UVs) {
   group=Icon
   parameterName=Texture
   samplerSource=SSM_Clamp_WorldGroupSettings
   texture=/Game/Athena/Items/Accolades/Materials/Textures/T_UI_accolades_Ram
  }
  n19 = Multiply(A=@n18.R, B=@n18.R)
  NamedRerouteDeclaration(None=SmoothStep(Min=@n13, Max=Add(A=@n13, B=@n14), Value=Multiply(A=@n19, B=@n17.0-1 mask)) {constMax=0.3, constMin=0.2}) {name=Icon, nodeColor=<rgba 0 0 0 1>}
  n20 = Constant()
  n21 = LinearInterpolate(
   B=ScalarParameter() {group=Icon, parameterName=IconOffsetYWithTiers}
   Alpha=Saturate(None=ScalarParameter() {defaultValue=2, desc=Support 0-3, group=Tiers, parameterName=Tier})
  )
  NamedRerouteDeclaration(None=Add(A=MaterialFunctionCall() {materialFunction=&GetUserInterfaceUV}.Normalized UV, B=AppendVector(A=@n20, B=@n21))) {name=IconOffset, nodeColor=<rgba 0 0 0 1>}
  NamedRerouteDeclaration(None=VectorParameter() {group=Color, parameterName=IconColor, defaultValue=<rgba 1 1 1 1>}.RGB) {name=IconColor, nodeColor=<rgba 0 0 0 1>}
  n22 = Constant() {r=1}
  n23 = ScalarParameter() {defaultValue=0.15, group=Tiers, parameterName=StarScale}
  NamedRerouteDeclaration(None=Floor(None=Divide(A=@n22, B=@n23))) {name=MaxStarsInRow, nodeColor=<rgba 0 0 0 1>}
  NamedRerouteDeclaration(None=LinearInterpolate(
    B=ScalarParameter() {defaultValue=1.15, group=Animation, parameterName=Anim_Highlight_Scale}
    Alpha=ScalarParameter() {group=Animation, parameterName=Anim_Highlight}
   ) {constA=1}) {name=TierUVScalar, nodeColor=<rgba 0 0 0 1>}
  n24 = MaterialFunctionCall() {materialFunction=&GetUserInterfaceUV}
  n25 = Add(
   A=MaterialFunctionCall(UVs=@n24.Normalized UV, Texture Scale=NamedRerouteUsage()) {materialFunction=&ScaleUVsByCenter}.UVs
   B=NamedRerouteUsage()
  ) {constB=0.5}
  n26 = NamedRerouteUsage()
  n27 = Constant() {r=1}
  n28 = Divide(B=NamedRerouteUsage())
  n29 = NamedRerouteUsage()
  n30 = OneMinus(None=Multiply(A=ScalarParameter() {defaultValue=2, desc=Support 0-3, group=Tiers, parameterName=Tier}, B=Multiply(A=@n28, B=@n29)))
  n31 = ComponentMask(None=MaterialFunctionCall!(Value=@n24.Normalized UV, Scale, Height) {materialFunction=/FortUILibrary/MaterialFunctions/Basics/MF_UI_PingPong}.Result) {r=true}
  n32 = Multiply(
   A=TextureSample(UVs=MaterialFunctionCall!(UV=@n25, Tiling X=@n26, Tiling Y=@n27) {materialFunction=/FortUILibrary/MaterialFunctions/Basics/MF_UI_Pattern}.Scaled Pattern UVs) {samplerSource=SSM_Clamp_WorldGroupSettings, texture=/Game/Athena/HUD/VolumePrompt/Materials/T_UI_StarDF}.R
   B=Ceil(None=SmoothStep(Min=@n30, Value=@n31) {constMin=0.6})
  )
  NamedRerouteDeclaration(None=StaticSwitchParameter(True=SmoothStep(Value=@n32) {constMax=0.2, constMin=0.15}, False=Constant()) {group=Tiers, parameterName=ShowStars}) {name=StarTexture, nodeColor=<rgba 0 0 0 1>}
  n33 = Divide(A=ScalarParameter() {defaultValue=0.15, group=Tiers, parameterName=StarScale})
  n34 = Fmod(A=ScalarParameter() {defaultValue=2, desc=Support 0-3, group=Tiers, parameterName=Tier}, B=Constant() {r=2})
  n35 = NamedRerouteUsage()
  n36 = LinearInterpolate(A=@n34, B=OneMinus(None=@n34), Alpha=Fmod(A=Floor(None=@n35), B=Constant() {r=2}))
  NamedRerouteDeclaration(None=AppendVector(A=Multiply(A=@n33, B=@n36), B=ScalarParameter() {defaultValue=0.375, group=Tiers, parameterName=StarOffsetY})) {name=StarCenterOffset, nodeColor=<rgba 0 0 0 1>}
 ]
}
