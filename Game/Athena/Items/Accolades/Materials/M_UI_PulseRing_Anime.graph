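# Additive UI material for an anime-style pulse ring in Athena accolades, drawing a glowing ring with a harder band inside it.
# Two exponential radial gradients of different radii are subtracted to form the ring, then multiplied by tiled, power-adjusted blurred noise to break it up.
# A smoothstep of that mask makes the hard band, which is tinted RingColor over a two-color glow gradient.
# The maximum of soft and hard masks forms the second output.
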Material /Game/Athena/Items/Accolades/Materials/M_UI_PulseRing_Anime
graph {
 properties {blendMode=BLEND_Additive, materialDomain=MD_UI, pixelDepthOffsetMode=PDOM_Legacy, usedShadingModels=Default Lit}
 nodes [
  n1 = Multiply(A=TextureCoordinate(), B=ScalarParameter() {defaultValue=1, parameterName=NoiseRepeat})
  n2 = ScalarParameter() {defaultValue=0.2, parameterName=HardRingRadius, sliderMax=0.5}
  n3 = ScalarParameter() {defaultValue=37, parameterName=HardRingDensity}
  n4 = Subtract(A=@n2, B=ScalarParameter() {defaultValue=0.1, parameterName=HardRingSize, sliderMax=0.05})
  n5 = Multiply(
   A=Power(
    Base=TextureSample(UVs=@n1) {samplerType=SAMPLERTYPE_Grayscale, texture=&LowResBlurredNoise}.R
    Exp=ScalarParameter() {parameterName=NoiseAmount}
   )
   B=Subtract(
    A=MaterialFunctionCall(Radius=@n2, Density=@n3) {materialFunction=&RadialGradientExponential}.RadialGradientExponential
    B=MaterialFunctionCall(Radius=@n4, Density=@n3) {materialFunction=&RadialGradientExponential}.RadialGradientExponential
   )
  )
  n6 = ScalarParameter() {defaultValue=0.003, parameterName=HardRingEdge}
  n7 = MaterialFunctionCall(Alpha=@n5, Min=@n6, Max=Add(A=@n6, B=ScalarParameter() {defaultValue=0.01, parameterName=EdgeSoftness})) {materialFunction=&SmoothStep}
  LinearInterpolate(
   A=LinearInterpolate(
    A=VectorParameter() {parameterName=GlowColor, defaultValue=<rgba 0.0249 0 0.151 1>}.RGB
    B=VectorParameter() {parameterName=GlowColor2, defaultValue=<rgba 0 0.3883 1 1>}.RGB
    Alpha=@n5
   )
   B=VectorParameter() {parameterName=RingColor, defaultValue=<rgba 0 0.8445 1 1>}.RGB
   Alpha=@n7.Result
  )
  Max(A=@n5, B=@n7.Result)
 ]
}
