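# UI translucent base material for XP accolade medals, with a perspective-warped medal texture, a reveal transition and an animated shine.
# A RevealAmount-driven vertical gradient thresholds blurred noise through SmoothStep, blending Texture through a white edge into TextureTop.
# A timed MF_Shine sweep takes the max of a flat pass and a pass bump-offset by the blended image's red channel, faking relief.
# The tinted shine is color-dodged over the blended medal.
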Material /Game/Athena/Items/Accolades/Materials/M_UI_XPMedal_Base
graph {
 properties {blendMode=BLEND_Translucent, materialDomain=MD_UI, pixelDepthOffsetMode=PDOM_Legacy, usedShadingModels=Default Lit}
 nodes [
  n1 = VectorParameter() {parameterName=ShineRange, defaultValue=<rgba 0 -1 0 1>}
  n2 = LinearInterpolate(A=@n1.R, B=@n1.R, Alpha=Divide(A=Time() {bOverride_Period=true, period=2}, B=ScalarParameter() {defaultValue=2, parameterName=ShineSpeed})) {constB=-1}
  n3 = ScalarParameter() {parameterName=ShineRotation}
  n4 = VectorParameter() {parameterName=RotationCenter, defaultValue=<rgba 0 0 0 1>}
  n5 = MaterialFunctionCall!(ShineOffset=@n2, Rotation=@n3, RotationCenter=@n4.RGB, UVs) {materialFunction=/Game/Athena/HUD/Materials/Victory/MF_Shine}
  n6 = ScalarParameter() {defaultValue=0.5, parameterName=BaseShine}
  n7 = VectorParameter() {group=Perspective, parameterName=PerspectiveOffset, defaultValue=<rgba 0 0 0 1>}
  n8 = VectorParameter() {group=Perspective, parameterName=Perspective}
  n9 = MaterialFunctionCall!(
   UV=Add(A=AppendVector(A=@n7.R, B=@n7.R), B=TextureCoordinate())
   Texture
   FixedScaleFactor_X
   FixedScaleFactor_Y
   Persp_Y=@n8.R
   Persp_X=@n8.R
   Slide_X=@n8.R
   Slide_Y=@n8.R
   Flip_Y
   Flip_X=StaticBoolParameter() {group=Perspective, parameterName=FlipX}
  ) {materialFunction=/Game/UI/Foundation/Materials/MF_UI_BasicPerspective}
  n10 = TextureSampleParameter2D(UVs=@n9.FinalUvs) {
   parameterName=Texture
   samplerSource=SSM_Clamp_WorldGroupSettings
   texture=/Game/Athena/Items/Accolades/Materials/Textures/Medals_Survivor_Bronze_128
  }
  n11 = Constant3Vector() {constant=<rgba 1 1 1 0>}
  n12 = TextureCoordinate()
  n13 = AppendVector(A=Constant(), B=LinearInterpolate(Alpha=ScalarParameter() {group=Reveal, parameterName=RevealAmount}) {constA=-1})
  n14 = OneMinus(None=ComponentMask(None=Add(A=@n12, B=@n13)) {g=true})
  n15 = Subtract(A=@n14, B=ScalarParameter() {defaultValue=0.1, group=Reveal, parameterName=RevealEdgeSize})
  n16 = ScalarParameter() {group=Reveal, parameterName=RevealEdgeSoftness}
  n17 = Add(A=@n15, B=@n16)
  n18 = TextureSampleParameter2D(UVs=Multiply(A=@n12, B=ScalarParameter() {defaultValue=1, group=Reveal, parameterName=RevealPatternRepeat})) {group=Reveal, parameterName=RevealNoise, samplerType=SAMPLERTYPE_Grayscale, texture=&LowResBlurredNoise}
  n19 = TextureSampleParameter2D(UVs=@n9.FinalUvs) {
   parameterName=TextureTop
   samplerSource=SSM_Clamp_WorldGroupSettings
   texture=/Game/Athena/Items/Accolades/Materials/Textures/Medals_Survivor_Gold_128
  }
  n20 = LinearInterpolate(
   A=LinearInterpolate(
    A=@n10.RGBA
    B=AppendVector(A=@n11, B=@n10.A)
    Alpha=MaterialFunctionCall(Alpha=@n18.R, Min=@n15, Max=@n17) {materialFunction=&SmoothStep}.Result
   )
   B=@n19.RGBA
   Alpha=MaterialFunctionCall(Alpha=@n18.R, Min=@n14, Max=Add(A=@n14, B=@n16)) {materialFunction=&SmoothStep}.Result
  )
  n21 = MaterialFunctionCall(Float4=@n20) {materialFunction=&BreakOutFloat4Components}
  n22 = MaterialFunctionCall!(
   HeightTexture=Desaturation(None=@n21.R)
   TextureCoordinates=TextureCoordinate()
   ReferencePlane=ScalarParameter() {defaultValue=0.5, parameterName=Ref}
   HeightRatioInput=ScalarParameter() {defaultValue=2.2, parameterName=BumpRatio}
   CameraVector=VectorParameter() {parameterName=CameraVector, defaultValue=<rgba -0.2692 0 0 1>}.RGB
  ) {materialFunction=/Engine/Functions/Engine_MaterialFunctions02/Utility/BumpOffset_Advanced.BumpOffset_advanced}
  n23 = VectorParameter() {parameterName=ShineColor, defaultValue=<rgba 1 0.803 0 1>}
  n24 = Multiply(
   A=Max(
    A=Multiply(A=@n5.Result, B=@n6)
    B=MaterialFunctionCall(ShineOffset=@n2, Rotation=@n3, RotationCenter=@n4.RGB, UVs=@n22.UVs) {materialFunction=&MF_Shine}.Result
   )
   B=@n23.A
  )
  Clamp(None=MaterialFunctionCall!(Blend=Multiply(A=@n24, B=@n23.RGB), Base=@n20) {materialFunction=/Engine/Functions/Engine_MaterialFunctions03/Blends/Blend_ColorDodge}.Result)
  Constant()
  MaterialFunctionCall() {materialFunction=&ScreenAlignedUVs}
  Multiply(A=@n9.0 or 1 Mask, B=@n21.A)
  Saturate(None=@n20)
  ScalarParameter() {parameterName=ShineOffset}
 ]
}
